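# Translucent unlit material for the glow trail on the OG Vault umbrella glider, set up for Niagara ribbons and mesh particles.
# A normalized instance-space Y direction is scaled by a power of U and the dynamic Bend parameter, forming an offset that curves along the trail.
# Particle alpha times inverted U gives a fade along the length.
# A Width parameter is declared but not used in the shown nodes.
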Material /BRCosmetics/Gadgets/Assets/VinderTech_Umbrella/Umbrella_OGVault/FX/M_Umbrealla_OGVault_GlowTrail
graph {
 properties {
  bUsedWithNiagaraMeshParticles=true
  bUsedWithNiagaraRibbons=true
  blendMode=BLEND_Translucent
  pixelDepthOffsetMode=PDOM_Legacy
  previewMesh=&/SM_Umbrealla_OGVault_CurvedRibbon
  shadingModel=MSM_Unlit
  usedShadingModels=Unlit
 }
 nodes [
  n1 = Transform(None=Constant3Vector() {constant=<rgba 0 1 0 0>}) {transformSourceType=TRANSFORMSOURCE_Instance}
  n2 = ComponentMask(None=TextureCoordinate()) {r=true}
  n3 = Power(Base=@n2)
  n4 = DynamicParameter() {defaultValue=<rgba 1 50 0 0.5>, paramNames [Width, Bend, "", ""]}
  Add(B=Multiply(A=Normalize(VectorInput=@n1), B=Multiply(A=@n3, B=@n4.Bend)))
  n5 = TextureCoordinate()
  Multiply(A=ParticleColor().A, B=OneMinus(None=ComponentMask(None=@n5) {r=true}))
  OneMinus(None=@n2)
 ]
}
